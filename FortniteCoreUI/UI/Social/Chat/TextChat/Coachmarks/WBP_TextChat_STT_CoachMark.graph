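# Sidebar onboarding coach mark that announces text chat text-to-speech and dictation accessibility features, wrapping WBP_UIKit_CoachMark.
# PreConstruct sets description and button sizing; OnShowTooltip activates the coach mark.
# Dismiss, other-action and OnEndTooltip paths notify close completion and deactivate both widgets, with other-action first broadcasting OnOtherActionButtonClicked.
# Separate functions set arrow position and open Communication settings at ShowTextToSpeech, choosing the tab ID by Save the World UI.
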WidgetBlueprint /FortniteCoreUI/UI/Social/Chat/TextChat/Coachmarks/WBP_TextChat_STT_CoachMark
graphs [
 EventGraph {nodes [
   "Button Dismiss Clicked (WBP_UIKit_CoachMark)"() {is=title K2Node_ComponentBoundEvent}
   > n1 = HandleDismissClicked(self:Fort Sidebar Onboard Tooltip Widget Object Reference)
   NotifyCloseAnimationFinished(execute=[@n2.then, @n1.then, @n3.then], self:Fort Sidebar Onboard Tooltip Widget Object Reference)
   > DeactivateWidget(self=Get WBP_UIKit_CoachMark() -> (WBP_UIKit_CoachMark:WBP UIKit Coach Mark Object Reference))
   > DeactivateWidget(self=Self-Reference())
   "Button Other Action Clicked (WBP_UIKit_CoachMark)"() {is=title K2Node_ComponentBoundEvent}
   > Call On Other Action Button Clicked!(execute:Exec, self:WBP Text Chat STT Coach Mark Object Reference) -> (then:Exec) {is=title K2Node_CallDelegate}
   > n2 = HandleOtherActionClicked(self:Fort Sidebar Onboard Tooltip Widget Object Reference)
   n3 = Event OnEndTooltip() {is=title K2Node_Event}
   Event PreConstruct()
   n4 = Get WBP_UIKit_CoachMark()
   > SetTextDescription!(
    execute
    self:WBP UIKit Coach Mark Base Object Reference=@n4
    Text:Text="LOCTEXT(\"Read incoming text chat messages aloud and dictate your replies. \\r\\n \\r\\nAvailable in Communication settings.\")"
   ) -> (then, Empty:bool)
   > SetSize!(
    execute
    self:Horizontal Box Slot Object Reference=SlotAsHorizontalBoxSlot!(Widget:Widget Object Reference=Get DismissButton!(self:WBP UIKit Coach Mark Base Object Reference=@n4) -> (DismissButton:WBP UIKit Button Regular Object Reference)) -> (ReturnValue:Horizontal Box Slot Object Reference)
    InSize_Value:float=1
    InSize_SizeRule:ESlateSizeRule Enum=Automatic
   ) -> (then)
   > SetSize(
    self=SlotAsHorizontalBoxSlot(Widget=Get OtherActionButton!(self:WBP UIKit Coach Mark Base Object Reference=@n4) -> (OtherActionButton:WBP UIKit Button Regular Object Reference))
    InSize_Value=1
    InSize_SizeRule=Fill
   )
   Event OnShowTooltip() {is=title K2Node_Event}
   > ActivateWidget(self=Get WBP_UIKit_CoachMark())
  ]}
 OnOtherActionButtonClicked {nodes [OnOtherActionButtonClicked() {is=title K2Node_FunctionEntry}]}
 OpenSettings {nodes [
   OpenSettings() {is=title K2Node_FunctionEntry}
   > RequestOpenSettings!(
    execute
    self:Fort Settings Context Object Reference=Get FortSettingsContext!() -> (ReturnValue:Fort Settings Context Object Reference) {is=title K2Node_GetBlueprintContextEx}
    ContextWidget:Common User Widget Object Reference=Self-Reference()
    SettingsTabId:Name=Select(Option 0:Name=CommunicationCollection, Option 1:Name=Communication, Index=IsSaveTheWorldUI!() -> (IsSaveTheWorld:bool)) -> (ReturnValue:Name)
    SettingToSelect:String=ShowTextToSpeech
   ) -> (then)
  ]}
 SetCoachmarkArrowPosition {nodes [
   n1 = SetCoachmarkArrowPosition!() -> (then, ArrowPosition:E_UI_OutBoardPosition Enum=NewEnumerator3) {is=title K2Node_FunctionEntry}
   > SetArrowPosition!(execute, self:WBP UIKit Coach Mark Base Object Reference=Get WBP_UIKit_CoachMark(), ArrowPosition:E_UI_OutBoardPosition Enum=@n1) -> (then)
  ]}
]
variables [OnOtherActionButtonClicked {}]
widgets {
 info {
  inheritedWidgetCount=2
  parentClass=/Script/FortniteUI.FortSidebarOnboardTooltipWidget
  rootWidgetClass=/FortUILibrary/UIKit/CoachMark/WBP_UIKit_CoachMark.WBP_UIKit_CoachMark_C
  widgetCount=3
 }
 widgets [
  WBP_UIKit_CoachMark {
   widgetClassPath=/FortUILibrary/UIKit/CoachMark/WBP_UIKit_CoachMark
   bIsVariable=true
   widget=WBP_UIKit_CoachMark
   properties {
    arrowPosition=Right_Down
    bIsFocusable=true
    buttonFillRule=Automatic
    navigation=&self:WidgetTree.WBP_UIKit_CoachMark.WidgetNavigation_0
    textDescription=Read incoming text chat messages aloud and dictate your replies. Available in Communication settings.
    textDismissButton=Got it
    textHeader=New accessibility features
    textOtherActionButton=Open Settings
    visibility=SelfHitTestInvisible
    widget Width=580
    in Decorator Classes [/FortniteCoreUI/UI/Social/Chat/TextChat/RichTextDecorator/RichText_TextChatDecorator.RichText_TextChatDecorator_C, null]
   }
  }
  Dismiss_Button {widgetClassPath=&CommonButtonLegacy, bInherited=true, widget=null}
  OtherAction_Button {widgetClassPath=&CommonButtonLegacy, bInherited=true, widget=null}
 ]
}
